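# Master material for the Supply Drop 2 gadget, layering World Position Offset animation onto the shared Fortnite weapon master function.
# A custom-primitive-data start time drives a curve-atlas lookup that deforms vertices above a height threshold.
# A static switch adds Emerald Glass wobble: timed landing bounce and sway, with vertex colors masking vertical movement and hand and hair rotations.
# Another switch routes the offset through vertex-color element hiding.
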Material /Game/Gadgets/Assets/Supply_Drop_2/Materials/M_Supply_Drop_Master
graph {
 properties {
  bAutomaticallySetUsageInEditor=false
  bUseMaterialAttributes=true
  bUsedWithClothing=true
  bUsedWithMeshDeformer=true
  bUsedWithMorphTargets=true
  bUsedWithNanite=true
  bUsedWithSkeletalMesh=true
  pixelDepthOffsetMode=PDOM_Legacy
  usedShadingModels=Default Lit
 }
 nodes [
  ComponentMask() {g=true, r=true}
  ComponentMask() {g=true, r=true}
  n1 = Add(A=PreSkinnedPosition(), B=Constant3Vector() {constant=<rgba 0 0 -356 0>})
  n2 = Constant3Vector() {constant=<rgba 0.5 0.5 -0.65 0>}
  n3 = Subtract(A=Time() {period=3}, B=ScalarParameter() {bUseCustomPrimitiveData=true, defaultValue=-100, parameterName=AnimTime})
  n4 = ScalarParameter() {defaultValue=1.5, parameterName=AnimDuration}
  n5 = CurveAtlasRowParameter(CurveTime=Saturate(None=Divide(A=@n3, B=@n4))) {
   atlas=&/Atlas_Supply_Drop
   curve=&/Curve_Supply_Drop_Anim
  }
  n6 = Multiply(A=@n5.R) {constB=2}
  n7 = ComponentMask(None=PreSkinnedPosition()) {b=true}
  n8 = Multiply(
   A=LinearInterpolate(A=Multiply(A=@n1, B=@n2) {constB=-1}, Alpha=NamedRerouteDeclaration(None=@n6) {name=Animation Time, nodeColor=<rgba 1 0.0938 0 1>})
   B=Saturate(None=Subtract(A=@n7) {constB=357})
  )
  Composite!(False=@n8) {subgraphName=EmeraldGreenWPO}
  MaterialExpressionReroute_0 = Reroute(None=@n8)
  n9 = StaticSwitchParameter(True=Add(A=NamedRerouteUsage(), B=NamedRerouteUsage()), False=@MaterialExpressionReroute_0.False) {parameterName=UseEmeraldGlassWobble}
  MaterialExpressionReroute_1 = Reroute(None=@n9)
  NamedRerouteDeclaration(None=@MaterialExpressionReroute_1.Output) {name=WPO, nodeColor=<rgba 0 1 0.6719 1>}
  n10 = Time() {period=3}
  n11 = ScalarParameter() {bUseCustomPrimitiveData=true, parameterName=EGW Bounce Animation}
  n12 = NamedRerouteDeclaration(None=Divide(A=Subtract(A=@n10, B=@n11), B=ScalarParameter() {defaultValue=2, parameterName=EGW Bounce Duration})) {name=EGW Animation Timeline, nodeColor=<rgba 1 0 0.7969 1>}
  n13 = Saturate(None=@n12)
  n14 = OneMinus(None=@n13)
  n15 = Sine(None=@n12) {period=0.8}
  NamedRerouteDeclaration(None=Multiply(A=@n14, B=Multiply(A=@n15) {constB=-1})) {name=EGW Hand Wobble Animation, nodeColor=<rgba 1 0 0.2578 1>}
  NamedRerouteDeclaration(None=@n13) {name=EGW Animation Timeline Saturated, nodeColor=<rgba 1 0 0.7969 1>}
  n16 = PreSkinnedPosition()
  NamedRerouteDeclaration(None=ShaderStageSwitch(PixelShader=VertexInterpolator(VS=@n16).PS, VertexShader=@n16)) {name=EGW Local Position, nodeColor=<rgba 1 0 0.0937 1>}
  n17 = VertexColor()
  NamedRerouteDeclaration(None=@n17.B) {name=EGW Vertical Mask, nodeColor=<rgba 0 0.6016 1 1>}
  n18 = Multiply(
   A=Multiply(A=ScalarParameter() {defaultValue=-60, parameterName=EGW Land Vertical Amp}, B=NamedRerouteUsage())
   B=Constant3Vector() {constant=<rgba 0 0 1 0>}
  )
  n19 = Time()
  n20 = Sine(None=@n19)
  n21 = LinearInterpolate(
   A=ScalarParameter() {defaultValue=1, parameterName=EGW Rotational Wobble Noise}
   B=OneMinus(None=ScalarParameter() {bUseCustomPrimitiveData=true, parameterName=EGW Stop High Frequency Wobble, primitiveDataIndex=1})
   Alpha=NamedRerouteUsage()
  ) {constB=0}
  n22 = Cosine(None=@n19)
  n23 = NamedRerouteDeclaration(None=AppendVector(A=Multiply(A=@n20, B=@n21) {constB=0}, B=Multiply(A=@n22, B=@n21) {constB=0})) {name=EGW Circular Vector, nodeColor=<rgba 1 0 0.2812 1>}
  n24 = Constant() {r=1}
  n25 = RotateAboutAxis(
   NormalizedRotationAxis=Normalize(VectorInput=AppendVector(A=@n23, B=@n24))
   RotationAngle=Multiply(A=Cosine(None=@n19) {period=1.103}, B=ScalarParameter() {defaultValue=0.0025, parameterName=EGW Main Wobble Amp})
   PivotPoint=Constant3Vector()
   Position=NamedRerouteUsage()
  )
  NamedRerouteDeclaration(None=Multiply(A=NamedRerouteUsage(), B=Add(A=@n18, B=@n25))) {name=EGW Vertical Movement, nodeColor=<rgba 1 0.8906 0 1>}
  NamedRerouteDeclaration(None=@n17.R) {name=EGW Hands Mask, nodeColor=<rgba 1 0 0.4922 1>}
  NamedRerouteDeclaration(None=@n17.G) {name=EGW Hair Mask, nodeColor=<rgba 1 0 0.4922 1>}
  n26 = Max(A=NamedRerouteUsage(), B=NamedRerouteUsage())
  n27 = AppendVector(A=Constant() {r=1}, B=Multiply(A=NamedRerouteUsage()) {constB=0.5})
  n28 = Time()
  n29 = Sine(None=@n28) {period=1.406}
  n30 = Sine(None=@n28) {period=1.103}
  n31 = NamedRerouteUsage()
  n32 = ScalarParameter() {defaultValue=0.004, parameterName=EGW Hands Wobble Amp}
  n33 = ScalarParameter() {defaultValue=0.003, parameterName=EGW Hair Wobble Amp}
  n34 = NamedRerouteUsage()
  n35 = LinearInterpolate(
   A=ScalarParameter() {defaultValue=0.06, parameterName=EGC Max Hand Bounce, sliderMin=-1}
   B=ScalarParameter() {defaultValue=0.03, parameterName=EGC Max Hair Bounce, sliderMin=-1}
   Alpha=NamedRerouteUsage()
  )
  n36 = Multiply(A=NamedRerouteUsage(), B=NamedRerouteUsage())
  n37 = ComponentMask(None=NamedRerouteUsage()) {g=true}
  n38 = Add(
   A=Multiply(A=LinearInterpolate(A=@n29, B=@n30, Alpha=@n31), B=LinearInterpolate(A=@n32, B=@n33, Alpha=@n34))
   B=Multiply(A=Multiply(A=@n35, B=@n36), B=Sign(None=@n37))
  )
  n39 = Constant() {r=1}
  n40 = Sign(None=ComponentMask(None=NamedRerouteUsage()) {g=true})
  n41 = NamedRerouteUsage()
  n42 = LinearInterpolate(
   A=Multiply(
    A=VectorParameter() {parameterName=EGW Hand Rotation Pivot, defaultValue=<rgba 12 136 450 0>}.RGB
    B=MaterialFunctionCall(X=@n39, Y=@n40, Z=@n39) {materialFunction=&MakeFloat3}.Result
   )
   B=VectorParameter() {parameterName=EGW Hair Rotation Pivot, defaultValue=<rgba -85 0 700 0>}.RGB
   Alpha=NamedRerouteDeclaration(None=Sign(None=@n41)) {name=EGW Hair Binary Mask, nodeColor=<rgba 1 0 0.0703 1>}
  ) {constB=1000}
  n43 = Add(A=NamedRerouteUsage(), B=NamedRerouteUsage())
  NamedRerouteDeclaration(None=Multiply(
    A=NamedRerouteDeclaration(None=@n26) {name=EGW Hands Hair Mask, nodeColor=<rgba 0 1 0.0625 1>}
    B=RotateAboutAxis(NormalizedRotationAxis=@n27, RotationAngle=@n38, PivotPoint=@n42, Position=@n43)
   )) {name=EGW Hands Hair Rotations, nodeColor=<rgba 1 0.4219 0 1>}
  n44 = Sine(None=@n13) {period=0.8}
  NamedRerouteDeclaration(None=Multiply(A=@n14, B=Abs(None=@n44))) {name=Bounce Animation, nodeColor=<rgba 0 0.3437 1 1>}
  PinBase() {reroutePins [False {expression=&self:MaterialExpressionReroute_0}]}
  PinBase!(Output=@n9) {reroutePins [Output {expression=&self:MaterialExpressionReroute_1}]}
  n45 = GetMaterialAttributes(None=MaterialFunctionCall!(
    Material=MaterialFunctionCall!(In) {materialFunction=/Game/Packages/Fortress_SharedMaterials/Base_Material_Functions/BaseCustomizationFunctions/MF_FN_Weapon_MASTER}.Material Result
    WorldPositionOffset=NamedRerouteUsage()
   ) {materialFunction=/Engine/Functions/MaterialLayerFunctions/MatLayerBlend_AddWorldPositionOffset}.Blended Material) {attributeGetTypes [#1]}
  SetMaterialAttributes!(
   MaterialAttributes=@n45.MaterialAttributes
   World Position Offset=StaticSwitchParameter(
    True=MaterialFunctionCall!(WPO) {materialFunction=/Game/Packages/Fortress_SharedMaterials/Base_Material_Functions/Fortnite_HideElementsWithVertexColor}.Result
    False=@n45.World Position Offset
   ) {group=Opacity Mask, parameterName=Use Vertex Colors for Mask}
  ) {attributeSetTypes [#1]}
 ]
}
